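# Unlit material for a bright vent light glow, in the VentsV2 mesh set of the SparklyJibblet special map.
# Its only node is a constant HDR light-blue color with values well above one, with no textures, parameters or animation.
# The graph lists no explicit link from that color to a material output.
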Material /Game/Athena/Apollo/Maps/Special/SparklyJibblet/Meshes/VentsV2/M_VentLightGlowBright
graph {
 properties {pixelDepthOffsetMode=PDOM_Legacy, shadingModel=MSM_Unlit, usedShadingModels=Unlit}
 nodes [Constant3Vector() {constant=<rgba 7.5 12.3188 20 0>}]
}
